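# Parent material for the GreenHazard pickaxe accessories, built on the 2023 character PreFX and PostFX functions.
# The Tech Art Mask green channel blends base color and emissive toward inverse-Fresnel green gradients.
# The same channel masks MetaBalls bubbles, borrowed from the GreenHazard body, which are layered over both colors before PostFX.
# A GalaMaiden function call and an emissive Add feed no other node shown.
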Material /BRCosmetics/Weapons/FORT_Melee/Pickaxe_GreenHazard/Materials/M_GreenHazard_Accesories_Parent
graph {
 properties {
  bUseMaterialAttributes=true
  bUsedWithMeshDeformer=true
  bUsedWithMorphTargets=true
  bUsedWithSkeletalMesh=true
  pixelDepthOffsetMode=PDOM_Legacy
  shadingModel=MSM_FromMaterialExpression
  usedShadingModels=Default Lit | Subsurface | Clear Coat | Subsurface Profile | Cloth
 }
 nodes [
  n1 = MaterialFunctionCall() {materialFunction=&MF_CharacterParent_2023_PreFX}
  n2 = GetMaterialAttributes(None=@n1.Result) {attributeGetTypes [#1, #2]}
  n3 = OneMinus(None=Fresnel(ExponentIn=Constant() {r=0.1}) {exponent=2})
  n4 = ComponentMask(None=@n1.Tech Art Mask) {g=true}
  n5 = LinearInterpolate(
   A=@n2.Emissive Color
   B=LinearInterpolate(A=Constant4Vector() {constant=<rgba 0.0083 0.05 0 1>}.RGB, B=Constant4Vector() {constant=<rgba 0.5 1 0 1>}.RGB, Alpha=@n3)
   Alpha=@n4
  )
  n6 = MaterialFunctionCall!(
   Background Parallax Offset
   Foreground Parallax Offset
   VerticalGradient=Constant() {r=1}
   Foreground Occlusion Mask=@n4
   Background Occlusion Mask=@n4
  ) {materialFunction=/BRCosmetics/Characters/Player/Male/Medium/Bodies/M_MED_GreenHazard/Materials/MF_MetaBalls_Bubbles}
  Add(A=@n5, B=Multiply(A=@n6.Result, B=@n6.Result))
  Constant()
  n7 = LinearInterpolate(
   A=@n2.Base Color
   B=LinearInterpolate(
    A=Constant4Vector() {constant=<rgba 0.0167 0.1 0 1>}.RGB
    B=VectorParameter() {parameterName=Param_3, defaultValue=<rgba 0.1 0.2 0 1>}.RGB
    Alpha=@n3
   )
   Alpha=@n4
  )
  MaterialFunctionCall(
   Material=SetMaterialAttributes!(
    MaterialAttributes=@n2.MaterialAttributes
    Base Color=LinearInterpolate(A=@n7, B=@n6.Result, Alpha=@n6.Result)
    Emissive Color=LinearInterpolate(A=@n5, B=@n6.Result, Alpha=@n6.Result)
   ) {attributeSetTypes [#1, #2]}
   M=@n1.M
   Pre Skinned Local Position=@n1.M
   Distance Blend=@n1.M
   Fresnel=@n1.M
  ) {materialFunction=&MF_CharacterParent_2023_PostFX}
  MaterialFunctionCall!(In=@n1.Result, TechArtMask=@n1.Result) {materialFunction=/BRCosmetics/Characters/Player/Female/Medium/Bodies/F_MED_GalaMaiden/Materials/MF_GalaMaiden}
 ]
}
