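# Control Rig for the GreenHazard pickaxe that procedurally animates its mechanism bones with random and noise-driven motion.
# Dyn_mechanism_03 spins toward lerp-smoothed random X angles with added noise jitter, while Dyn_mechanism_02 sways on Y from slower noise.
# A rounded random 0/1 at varying intervals, Kalman-smoothed, weights opposing rotations on the Up and Down bones, snapping them apart or together.
# The branch's >= 0 test always passes, and one noise node is unconnected.
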ControlRigBlueprint /BRCosmetics/Weapons/FORT_Melee/Pickaxe_GreenHazard/Meshes/GreenHazard_Axe_CtrlRig
graphs [
 RigVMFunctionLibrary {links [], nodes []}
 RigVMModel {nodes [
   RigUnit_BeginExecution = &RigUnit_BeginExecution()
   RigVMFunction_Sequence = &RigVMFunction_Sequence(ExecuteContext=@RigUnit_BeginExecution.ExecutePin)
   n1 = /Script/RigVM.RigVMFunction_RandomFloat(
    Maximum=720
    Minimum=-360
    Seed=30
    Duration=&RigVMFunction_RandomFloat(Duration=3.5, Maximum=3.5, Minimum=1, Seed=30) {template="Random::Execute(in Seed,in Minimum,in Maximum,in Duration,out Result)"}.Result
   ) {is=struct RigVMUnitNode, template="Random::Execute(in Seed,in Minimum,in Maximum,in Duration,out Result)"}
   RigVMFunction_MathQuaternionFromEuler = &RigVMFunction_MathQuaternionFromEuler(
    Euler=(X=0,Y=0,Z=0)
    RotationOrder=XYZ
    Euler.X=&RigVMFunction_MathDoubleAdd(
     A=/Script/RigVM.RigVMFunction_AccumulateVectorLerp(Blend=0.05, InitialValue=(X=0,Y=0,Z=0), TargetValue=(X=0,Y=0,Z=0), bIntegrateDeltaTime=false, TargetValue.X=@n1.Result) {is=struct RigVMUnitNode, template="AccumulateLerp::Execute(in TargetValue,in InitialValue,in Blend,in bIntegrateDeltaTime,out Result)"}.Result.X
     B=/Script/RigVM.RigVMFunction_NoiseDouble(Frequency=5, Maximum=5, Minimum=-5, Speed=3.5, Value=0) {is=struct RigVMUnitNode, template="Noise::Execute(in Value,in Speed,in Frequency,in Minimum,in Maximum,out Result)"}.Result
    ).Result
   )
   &RigUnit_ModifyTransforms(
    ItemToModify=((Item=(Type=Bone,Name="Dyn_mechanism_03"),Transform=(Rotation=(X=0,Y=0,Z=0,W=1),Translation=(X=0,Y=0,Z=0),Scale3D=(X=1,Y=1,Z=1))))
    Weight=1
    WeightMinimum=0
    ItemToModify.0.Transform.Rotation=@RigVMFunction_MathQuaternionFromEuler.Result
    ExecutePin=@RigVMFunction_Sequence.A
   )
   Add = &RigVMFunction_MathDoubleAdd(
    A=0
    B=&RigVMFunction_NoiseDouble(Frequency=5, Maximum=15, Minimum=-20, Speed=0.25, Value=0) {template="Noise::Execute(in Value,in Speed,in Frequency,in Minimum,in Maximum,out Result)"}.Result
   )
   > &RigUnit_ModifyTransforms(
    ItemToModify=((Item=(Type=Bone,Name="Dyn_mechanism_02"),Transform=(Rotation=(X=0,Y=0,Z=0,W=1),Translation=(X=0,Y=0,Z=0),Scale3D=(X=1,Y=1,Z=1))))
    Weight=1
    WeightMinimum=0
    ItemToModify.0.Transform.Rotation=&RigVMFunction_MathQuaternionFromEuler(Euler=(X=0,Y=0,Z=0), RotationOrder=XYZ, Euler.Y=@Add.Result).Result
   )
   Round = /Script/RigVM.RigVMFunction_MathDoubleRound(Value=&RigVMFunction_RandomFloat(
     Maximum=1
     Minimum=0
     Seed=30
     Duration=&RigVMFunction_RandomFloat(Duration=0.5, Maximum=2.5, Minimum=1, Seed=30) {template="Random::Execute(in Seed,in Minimum,in Maximum,in Duration,out Result)"}.Result
    ) {template="Random::Execute(in Seed,in Minimum,in Maximum,in Duration,out Result)"}.Result) {is=struct RigVMUnitNode, template="Round::Execute(in Value,out Result,out Int)"}
   GreaterEqual = /Script/RigVM.RigVMFunction_MathDoubleGreaterEqual(B=0, A=@Round.Result) {is=struct RigVMUnitNode, template="GreaterEqual::Execute(in A,in B,out Result)"}
   RigVMFunction_ControlFlowBranch = &RigVMFunction_ControlFlowBranch(Condition=@GreaterEqual.Result, ExecuteContext=@RigVMFunction_Sequence.B)
   KalmanFilter = &RigVMFunction_KalmanFloat(BufferSize=4, Value=@Round.Result) {template="KalmanFilter::Execute(in Value,in BufferSize,out Result)"}
   &RigUnit_ModifyTransforms(
    ItemToModify="((Item=(Type=Bone,Name=\"Dyn_mechanism_Down\"),Transform=(Rotation=(X=0,Y=-0.1564,Z=0,W=0.9877),Translation=(X=0,Y=0,Z=0),Scale3D=(X=1,Y=1,Z=1))),(Item=(Type=Bone,Name=\"Dyn_mechanism_Up\"),Transform=(Rotation=(X=0,Y=0.1564,Z=0,W=0.9877),Translation=(X=0,Y=0,Z ...(truncated)"
    WeightMinimum=0
    Weight=@KalmanFilter.Result
    ExecutePin=@RigVMFunction_ControlFlowBranch.True
   )
   &RigVMFunction_NoiseDouble(Frequency=4, Maximum=0.5, Minimum=-0.5, Speed=5, Value=0) {template="Noise::Execute(in Value,in Speed,in Frequency,in Minimum,in Maximum,out Result)"}
  ]}
]
hierarchy [
 b_AR_Root {loc=<0 0 0>, rot=<0 0 0.7071 0.7071>}
 Melee_TwoHanded_Handle {parent=b_AR_Root, loc=<8.1477 -0.008 -29.0793>, rot=<0 0 0.7071 -0.7071>}
 Dyn_wire_Fr_01 {parent=Melee_TwoHanded_Handle, loc=<0.7729 -1.8224 -38.3223>, rot=<0.5 -0.5 -0.5 -0.5>}
 Dyn_wire_Fr_02 {parent=Dyn_wire_Fr_01, loc=<7.7375 0 0>, rot=<0 0 0 1>}
 Dyn_wire_Bk_01 {parent=Melee_TwoHanded_Handle, loc=<0.7729 -0.1988 -38.3223>, rot=<0.5 -0.5 -0.5 -0.5>}
 Dyn_wire_Bk_02 {parent=Dyn_wire_Bk_01, loc=<7.7375 0 0>, rot=<0 0 0 1>}
 Dyn_mechanism_01 {parent=Melee_TwoHanded_Handle, loc=<0.7729 -1.0413 69.1371>, rot=<0.3566 0.3566 0.6106 -0.6106>}
 Dyn_mechanism_02 {parent=Dyn_mechanism_01, loc=<30.0158 0 9.1984>, rot=<0.8821 0 0.4711 0>}
 Dyn_mechanism_03 {parent=Dyn_mechanism_02, loc=<11.7806 0 -19.8306>, rot=<0 -0.7221 0 -0.6918>}
 Dyn_mechanism_04 {parent=Dyn_mechanism_03, loc=<5.1414 0 -0.285>, rot=<0 -0.04 0 -0.9992>}
 Dyn_mechanism_Up {parent=Dyn_mechanism_04, loc=<0 0 0>, rot=<0 0 0 1>}
 Dyn_mechanism_Down {parent=Dyn_mechanism_04, loc=<0 0 0>, rot=<0 0 0 1>}
]
variables [
 GetTransform_Transform_Rotation {type=FQuat, default=(X=0,Y=0,Z=0,W=1)}
 NewVar {type=int32, default=0}
 RigVMFunction_MathQuaternionToEuler_Result_X {type=double, default=0}
 Speed {type=double, default=2}
 Time {type=double, default=0}
 Tr {type=FQuat, default=(X=0,Y=0,Z=0,W=1)}
]
