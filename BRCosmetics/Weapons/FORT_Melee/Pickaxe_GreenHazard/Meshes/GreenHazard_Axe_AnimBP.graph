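# Post-input AnimBP for the GreenHazard pickaxe that adds secondary physics motion to its front and back wires and mechanism bones.
# Blends by bIsInFrontEnd between two AnimDynamics chains from the cached input pose: world-space wires in front end, root-relative mechanism plus tighter wires otherwise.
# Two rotation-limited Trail controllers then act on the wires only when CurrentLOD is 2 or higher, as set each update.
# The GreenHazard_Axe_CtrlRig Control Rig produces the final pose.
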AnimBlueprint /BRCosmetics/Weapons/FORT_Melee/Pickaxe_GreenHazard/Meshes/GreenHazard_Axe_AnimBP
target_skeleton=&/GreenHazard_Axe_Skeleton
graphs [
 AnimGraph {nodes [
   n1 = Anim Dynamics - Bone: Dyn_wire_Fr_01(ComponentPose=Local To Component(LocalPose=Use cached pose 'Base'()), Alpha=1, ExternalForce="0,0,0") {
    angularDampingOverride=0.999
    angularSpringConstant=500
    bAngularSpring=true
    bOverrideAngularDamping=true
    bUseGravityOverride=true
    simulationSpace=World
    boundBone {boneName=Dyn_wire_Fr_01}
    componentAppliedLinearAccClamp=<1000 1000 1000>
    componentLinearAccScale=<0.5 0.5 0.5>
    componentLinearVelScale=<0.5 0.5 0.5>
    gravityOverride {z=-980}
    physicsBodyDefinitions [{
      boundBone {boneName=Dyn_wire_Fr_01}
      boxExtents=<12 12 12>
      constraintSetup {angularLimitsMax=<10 50 90>, angularLimitsMin=<-10 -50 -20>, angularTarget=<1 0 0>}
      localJointOffset=<0.5 0 0>
     }]
   }
   n2 = Anim Dynamics - Bone: Dyn_wire_Fr_02(ComponentPose=@n1, Alpha=1, ExternalForce="0,0,0") {
    angularDampingOverride=0.999
    angularSpringConstant=500
    bAngularSpring=true
    bOverrideAngularDamping=true
    bUseGravityOverride=true
    simulationSpace=World
    boundBone {boneName=Dyn_wire_Fr_02}
    componentAppliedLinearAccClamp=<1000 1000 1000>
    componentLinearAccScale=<0.5 0.5 0.5>
    componentLinearVelScale=<0.5 0.5 0.5>
    gravityOverride {z=-980}
    physicsBodyDefinitions [{
      boundBone {boneName=Dyn_wire_Fr_02}
      boxExtents=<12 12 12>
      constraintSetup {angularLimitsMax=<10 50 90>, angularLimitsMin=<-10 -50 -20>, angularTarget=<1 0 0>}
      localJointOffset=<0.5 0 0>
     }]
   }
   n3 = Anim Dynamics - Bone: Dyn_wire_Bk_01(ComponentPose=@n2, Alpha=1, ExternalForce="0,0,0") {
    angularDampingOverride=0.999
    angularSpringConstant=500
    bAngularSpring=true
    bOverrideAngularDamping=true
    bUseGravityOverride=true
    simulationSpace=World
    boundBone {boneName=Dyn_wire_Bk_01}
    componentAppliedLinearAccClamp=<1000 1000 1000>
    componentLinearAccScale=<0.5 0.5 0.5>
    componentLinearVelScale=<0.5 0.5 0.5>
    gravityOverride {z=-980}
    physicsBodyDefinitions [{
      boundBone {boneName=Dyn_wire_Bk_01}
      boxExtents=<12 12 12>
      constraintSetup {angularLimitsMax=<10 50 90>, angularLimitsMin=<-10 -50 -20>, angularTarget=<1 0 0>}
      localJointOffset=<0.5 0 0>
     }]
   }
   n4 = Anim Dynamics - Bone: Dyn_wire_Bk_02(ComponentPose=@n3, Alpha=1, ExternalForce="0,0,0") {
    angularDampingOverride=0.999
    angularSpringConstant=500
    bAngularSpring=true
    bOverrideAngularDamping=true
    bUseGravityOverride=true
    simulationSpace=World
    boundBone {boneName=Dyn_wire_Bk_02}
    componentAppliedLinearAccClamp=<1000 1000 1000>
    componentLinearAccScale=<0.5 0.5 0.5>
    componentLinearVelScale=<0.5 0.5 0.5>
    gravityOverride {z=-980}
    physicsBodyDefinitions [{
      boundBone {boneName=Dyn_wire_Bk_02}
      boxExtents=<12 12 12>
      constraintSetup {angularLimitsMax=<10 50 90>, angularLimitsMin=<-10 -50 -20>, angularTarget=<1 0 0>}
      localJointOffset=<0.5 0 0>
     }]
   }
   n5 = Component To Local(ComponentPose=@n4)
   n6 = Anim Dynamics - Bone: Dyn_mechanism_01(ComponentPose=Local To Component(LocalPose=Use cached pose 'Base'()), Alpha=1, ExternalForce="0,0,0") {
    angularDampingOverride=0.9999
    angularSpringConstant=1600
    bAngularSpring=true
    bOverrideAngularDamping=true
    bUseGravityOverride=true
    simulationSpace=RootRelative
    boundBone {boneName=Dyn_mechanism_01}
    componentAppliedLinearAccClamp=<700 700 700>
    componentLinearAccScale=<0.4 0.4 0.4>
    componentLinearVelScale=<0.4 0.4 0.4>
    gravityOverride {z=-980}
    physicsBodyDefinitions [{
      boundBone {boneName=Dyn_mechanism_01}
      boxExtents=<10 10 10>
      constraintSetup {angularLimitsMax=<0 12 0>, angularLimitsMin=<0 -12 0>, angularTarget=<1 0 0>}
      localJointOffset=<0.3 0 0>
     }]
   }
   n7 = Anim Dynamics - Bone: Dyn_mechanism_02(ComponentPose=@n6, Alpha=1, ExternalForce="0,0,0") {
    angularDampingOverride=0.9999
    angularSpringConstant=1600
    bAngularSpring=true
    bOverrideAngularDamping=true
    bUseGravityOverride=true
    simulationSpace=RootRelative
    boundBone {boneName=Dyn_mechanism_02}
    componentAppliedLinearAccClamp=<700 700 700>
    componentLinearAccScale=<0.4 0.4 0.4>
    componentLinearVelScale=<0.4 0.4 0.4>
    gravityOverride {z=-980}
    physicsBodyDefinitions [{
      boundBone {boneName=Dyn_mechanism_02}
      boxExtents=<10 10 10>
      constraintSetup {angularLimitsMax=<0 9 0>, angularLimitsMin=<0 -9 0>, angularTarget=<1 0 0>}
      localJointOffset=<0.3 0 0>
     }]
   }
   n8 = Anim Dynamics - Bone: Dyn_wire_Fr_01(ComponentPose=@n7, Alpha=1, ExternalForce="0,0,0") {
    angularDampingOverride=0.999
    angularSpringConstant=500
    bAngularSpring=true
    bOverrideAngularDamping=true
    bUseGravityOverride=true
    simulationSpace=RootRelative
    boundBone {boneName=Dyn_wire_Fr_01}
    componentAppliedLinearAccClamp=<600 600 600>
    componentLinearAccScale=<0.3 0.3 0.3>
    componentLinearVelScale=<0.3 0.3 0.3>
    gravityOverride {z=-980}
    physicsBodyDefinitions [{
      boundBone {boneName=Dyn_wire_Fr_01}
      boxExtents=<10 10 10>
      constraintSetup {angularLimitsMax=<10 10 90>, angularLimitsMin=<-10 -10 -20>, angularTarget=<1 0 0>}
      localJointOffset=<0.12 -1 0>
     }]
   }
   n9 = Anim Dynamics - Bone: Dyn_wire_Fr_02(ComponentPose=@n8, Alpha=1, ExternalForce="0,0,0") {
    angularDampingOverride=0.999
    angularSpringConstant=500
    bAngularSpring=true
    bOverrideAngularDamping=true
    bUseGravityOverride=true
    simulationSpace=RootRelative
    boundBone {boneName=Dyn_wire_Fr_02}
    componentAppliedLinearAccClamp=<600 600 600>
    componentLinearAccScale=<0.3 0.3 0.3>
    componentLinearVelScale=<0.3 0.3 0.3>
    gravityOverride {z=-980}
    physicsBodyDefinitions [{
      boundBone {boneName=Dyn_wire_Fr_02}
      boxExtents=<10 10 10>
      constraintSetup {angularLimitsMax=<10 12 90>, angularLimitsMin=<-10 -12 -20>, angularTarget=<1 0 0>}
      localJointOffset=<0.12 -1 0>
     }]
   }
   n10 = Anim Dynamics - Bone: Dyn_wire_Bk_01(ComponentPose=@n9, Alpha=1, ExternalForce="0,0,0") {
    angularDampingOverride=0.999
    angularSpringConstant=500
    bAngularSpring=true
    bOverrideAngularDamping=true
    bUseGravityOverride=true
    simulationSpace=RootRelative
    boundBone {boneName=Dyn_wire_Bk_01}
    componentAppliedLinearAccClamp=<600 600 600>
    componentLinearAccScale=<0.3 0.3 0.3>
    componentLinearVelScale=<0.3 0.3 0.3>
    gravityOverride {z=-980}
    physicsBodyDefinitions [{
      boundBone {boneName=Dyn_wire_Bk_01}
      boxExtents=<10 10 10>
      constraintSetup {angularLimitsMax=<10 10 90>, angularLimitsMin=<-10 -10 -20>, angularTarget=<1 0 0>}
      localJointOffset=<0.12 -1 0>
     }]
   }
   n11 = Anim Dynamics - Bone: Dyn_wire_Bk_02(ComponentPose=@n10, Alpha=1, ExternalForce="0,0,0") {
    angularDampingOverride=0.999
    angularSpringConstant=500
    bAngularSpring=true
    bOverrideAngularDamping=true
    bUseGravityOverride=true
    simulationSpace=RootRelative
    boundBone {boneName=Dyn_wire_Bk_02}
    componentAppliedLinearAccClamp=<600 600 600>
    componentLinearAccScale=<0.3 0.3 0.3>
    componentLinearVelScale=<0.3 0.3 0.3>
    gravityOverride {z=-980}
    physicsBodyDefinitions [{
      boundBone {boneName=Dyn_wire_Bk_02}
      boxExtents=<10 10 10>
      constraintSetup {angularLimitsMax=<10 10 90>, angularLimitsMin=<-10 -10 -20>, angularTarget=<1 0 0>}
      localJointOffset=<0.12 -1 0>
     }]
   }
   n12 = Component To Local(ComponentPose=@n11)
   n13 = Get bIsInFrontEnd()
   n14 = Get Lod 2 and above() -> (Lod 2 and above:bool)
   n15 = Trail controller - Bone: Dyn_wire_Fr_02!(
    ComponentPose=Local To Component(LocalPose=Blend Poses by bool(BlendPose_0=@n5, BlendPose_1=@n12, bActiveValue=@n13))
    bAlphaBoolEnabled:bool=@n14
    RelaxationSpeedScale:float=1
   ) -> (Pose) {
    alphaInputType=Bool
    bLimitRotation=true
    bShowTrailLocation=true
    chainLength=5
    maxDeltaTime=0.0333
    alphaScaleBiasClamp {bInterpResult=true}
    baseJoint {boneName=Dyn_wire_Fr_01}
    fakeVelocity {z=150}
    relaxationSpeedScaleInputProcessor {bInterpResult=true}
    rotationLimits [
     {limitMax=<15 12 12>, limitMin=<-15 -12 -12>}
     {limitMax=<15 12 12>, limitMin=<-15 -12 -12>}
     {limitMax=<15 12 12>, limitMin=<-15 -12 -12>}
     {limitMax=<15 12 12>, limitMin=<-15 -12 -12>}
     {limitMax=<15 12 12>, limitMin=<-15 -12 -12>}
    ]
    rotationOffsets [<0 0 0>, <0 0 0>, <0 0 0>, <0 0 0>, <0 0 0>]
    trailBone {boneName=Dyn_wire_Fr_02}
    trailRelaxationSpeed {editorCurveData {keys [<at 0 10 RCIM_Linear>,<at 0.4971 7 RCIM_Linear>,<at 1 4 RCIM_Linear>]}}
   }
   n16 = Trail controller - Bone: Dyn_wire_Bk_02!(ComponentPose=@n15, bAlphaBoolEnabled:bool=@n14, RelaxationSpeedScale:float=1) -> (Pose) {
    alphaInputType=Bool
    bLimitRotation=true
    bShowTrailLocation=true
    chainLength=6
    maxDeltaTime=0.0333
    alphaScaleBiasClamp {bInterpResult=true}
    baseJoint {boneName=Dyn_wire_Bk_01}
    fakeVelocity {z=150}
    relaxationSpeedScaleInputProcessor {bInterpResult=true}
    rotationLimits [
     {limitMax=<12 12 12>, limitMin=<-12 -12 -12>}
     {limitMax=<12 12 12>, limitMin=<-12 -12 -12>}
     {limitMax=<12 12 12>, limitMin=<-12 -12 -12>}
     {limitMax=<12 20 12>, limitMin=<-12 -20 -12>}
     {limitMax=<12 20 12>, limitMin=<-12 -20 -12>}
     {limitMax=<12 20 12>, limitMin=<-12 -20 -12>}
    ]
    rotationOffsets [<0 0 0>, <0 0 0>, <0 0 0>, <0 0 0>, <0 0 0>, <0 0 0>]
    trailBone {boneName=Dyn_wire_Bk_02}
    trailRelaxationSpeed {editorCurveData {keys [<at 0 10 RCIM_Linear>,<at 0.5 7 RCIM_Linear>,<at 1 4 RCIM_Linear>]}}
   }
   n17 = Control Rig(Source=Component To Local(ComponentPose=@n16)) {controlRigAssetReference {blueprintRigClass=&/GreenHazard_Axe_CtrlRig.GreenHazard_Axe_CtrlRig_C}}
   Output Pose(Result=@n17) {layerGroup=None}
   AnimGraphNode_SubInput_0 = Input Pose()
   Save cached pose 'Base'(Pose=@AnimGraphNode_SubInput_0) {cacheName=Base}
  ]}
 EventGraph {nodes [
   Event BlueprintUpdateAnimation()
   > Set Lod 2 and above(Lod 2 and above:bool="integer >= integer"(A=Get CurrentLOD(), B=2)) -> (Output_Get:bool)
   TryGetPawnOwner()
  ]}
]
variables [Lod 2 and above {type=boolean}]
